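# Masked material for the RL105 vehicle wheel, layering an animated "brain pulse" look over the shared MF_WheelParent_Simple_Alpha function.
# A time- and UV-driven cosine pushes red-vertex-color areas along their normals, with a panning detail normal added there.
# Base color combines Fresnel, AO/cavity tint and masked light spots, with the emissive texture blending in RimLightColor.
# The overridden attributes then pass through vehicle damage and de-rez transition functions.
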Material /VehicleCosmetics/Wheels/RL105/Materials/M_Wheel_RL105
graph {
 properties {
  bAutomaticallySetUsageInEditor=false
  bUseMaterialAttributes=true
  bUsedWithMeshDeformer=true
  bUsedWithSkeletalMesh=true
  blendMode=BLEND_Masked
  pixelDepthOffsetMode=PDOM_Legacy
  usedShadingModels=Default Lit
 }
 nodes [
  Constant3Vector() {constant=<rgba 0 0 1 0>}
  n1 = MaterialFunctionCall() {materialFunction=/VehicleCosmetics/Wheels/Mage/Materials/MF_WheelParent_Simple_Alpha}
  n2 = SetMaterialAttributes!(
   MaterialAttributes=@n1.MaterialAttributes
   World Position Offset=NamedRerouteUsage()
   Normal=NamedRerouteUsage()
   Base Color=NamedRerouteUsage()
   Specular=Constant3Vector() {constant=<rgba 0.0743 0.3906 0 1>}
  ) {attributeSetTypes [
    #1
    #2
    #3
    #4
   ]}
  MaterialFunctionCall(In=MaterialFunctionCall(Material=@n2) {materialFunction=&MF_BaseVehicle_Damage}.Result) {materialFunction=&MF_VehicleDeRez_Transition}
  n3 = Constant3Vector() {constant=<rgba 0.1769 1 0 1>}
  Multiply(A=Constant() {r=2}, B=@n3)
  n4 = VertexNormalWS()
  n5 = Time()
  n6 = ScalarParameter() {defaultValue=0.5, group=SE_Param, parameterName=Brain_Pulse_Speed}
  n7 = TextureCoordinate() {coordinateIndex=2, uTiling=2, vTiling=2}
  n8 = Cosine(None=Add(A=Multiply(A=@n5, B=@n6), B=ComponentMask(None=@n7) {g=true}))
  Multiply(A=@n4, B=@n8)
  n9 = Constant() {r=1}
  n10 = Multiply(A=@n4, B=Divide(A=Add(A=@n8, B=@n9), B=Constant() {r=2}))
  n11 = ScalarParameter() {defaultValue=1.1, parameterName=Brain_Pulse_Amplitude}
  n12 = VertexColor()
  NamedRerouteDeclaration(None=LinearInterpolate(A=Constant(), B=Multiply(A=@n10, B=@n11), Alpha=AppendVector(A=@n12.R, B=@n12.R))) {name=WPO, nodeColor=<rgba 0 0.1363 1 1>}
  n13 = NamedRerouteUsage()
  n14 = Fresnel() {exponent=1}
  n15 = Add(A=@n14, B=Power(Base=@n14, Exp=Constant() {r=2}))
  n16 = Power(Base=@n14, Exp=Constant() {r=3})
  n17 = TextureSampleParameter2D() {
   group=SE_Param
   parameterName=AO
   samplerType=SAMPLERTYPE_LinearColor
   sortPriority=2
   texture=/VehicleCosmetics/Wheels/RL105/Texture/T_RL105_Rim_AO
  }
  n18 = TextureSampleParameter2D() {
   group=SE_Param
   parameterName=CAV
   samplerType=SAMPLERTYPE_LinearColor
   sortPriority=2
   texture=/VehicleCosmetics/Wheels/RL105/Texture/T_RL105_Rim_CAV
  }
  n19 = Multiply(A=@n17.RGB, B=@n18.RGB)
  n20 = Multiply(A=Constant() {r=2}, B=Multiply(A=Add(A=@n15, B=@n16), B=@n19))
  n21 = Clamp(None=ComponentMask(None=NamedRerouteUsage()) {g=true})
  n22 = ScalarParameter() {defaultValue=5, group=SE_Param, parameterName=LightSpots}
  n23 = Add(
   A=Multiply(
    A=Add(A=@n20, B=@n19)
    B=VectorParameter() {group=SE_Param, parameterName=CustomColor, sortPriority=2, defaultValue=<rgba 0.0592 0.4062 0 1>}.RGB
   )
   B=Multiply(A=Multiply(A=@n21, B=@n22), B=@n3)
  )
  n24 = ComponentMask(None=NamedRerouteUsage()) {r=true}
  n25 = Multiply(A=Multiply(A=Clamp(None=@n24), B=Constant() {r=5}), B=Constant3Vector() {constant=<rgba 0 0.0104 0.0053 1>})
  n26 = VectorParameter() {
   desc=Color fill for R value of RGB Mask
   group=2_Rim Properties
   parameterName=Rim_ColorA
   channelNames {a=Unused, b="", g="", r=""}
   defaultValue=<rgba 1 0 0 1>
  }
  n27 = ComponentMask(None=NamedRerouteUsage()) {r=true}
  n28 = Multiply(A=@n27, B=Add(A=Add(A=@n23, B=@n25), B=@n26))
  n29 = VectorParameter() {group=SE_Param, parameterName=RimLightColor, defaultValue=<rgba 0.3429 0 0.375 1>}
  n30 = TextureSampleParameter2D() {
   group=SE_Param
   parameterName=Emissive
   samplerType=SAMPLERTYPE_LinearColor
   sortPriority=2
   texture=/VehicleCosmetics/Wheels/RL105/Texture/T_RL105_Rim_E
  }
  NamedRerouteDeclaration(None=LinearInterpolate(A=LinearInterpolate(A=@n13, B=@n28, Alpha=@n27), B=@n29.RGB, Alpha=@n30.RGB)) {name=Fix, nodeColor=<rgba 0.4297 1 0 1>}
  NamedRerouteDeclaration(None=@n1.NormalTire) {name=Normal_Tire, nodeColor=<rgba 0.5372 0.3698 1 1>}
  n31 = NamedRerouteUsage()
  n32 = TextureSampleParameter2D(UVs=Panner(Coordinate=TextureCoordinate()) {speedX=0.008}) {
   group=SE_Param
   parameterName=Normal_SE
   samplerType=SAMPLERTYPE_Normal
   sortPriority=2
   texture=/VehicleCosmetics/Wheels/RL105/Texture/T_RLWH105_Normal_thread_SE
  }
  n33 = Multiply(A=@n32.RGB, B=Constant3Vector() {constant=<rgba 2 2 0 0>})
  NamedRerouteDeclaration(None=LinearInterpolate(A=NamedRerouteUsage(), B=Add(A=@n31, B=@n33), Alpha=VertexColor().R)) {name=Normal, nodeColor=<rgba 0.0646 0 1 1>}
  NamedRerouteDeclaration(None=@n1.NormalRim) {name=Normal_Rim, nodeColor=<rgba 1 0.3191 0.6989 1>}
  NamedRerouteDeclaration(None=@n1.RimRGBAMask) {name=RimRGB, nodeColor=<rgba 0 1 0.7188 1>}
  NamedRerouteDeclaration(None=GetMaterialAttributes(None=@n1.MaterialAttributes) {attributeGetTypes [#3]}.Base Color) {name=BaseColorIN, nodeColor=<rgba 0 0.625 1 1>}
  NamedRerouteDeclaration(None=Multiply(A=Multiply(A=@n30.RGB, B=@n29.RGB), B=ScalarParameter() {defaultValue=1, group=SE_Param, parameterName=RimLightColor})) {name=Emissive, nodeColor=<rgba 1 0 0.8985 1>}
  NamedRerouteDeclaration(None=@n32.RGB) {name=NormalPanner, nodeColor=<rgba 0 1 1 1>}
  NamedRerouteUsage()
  WorldPosition()
 ]
}
